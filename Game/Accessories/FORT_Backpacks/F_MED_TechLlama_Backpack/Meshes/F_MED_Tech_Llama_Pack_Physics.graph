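# Physics asset for the Tech Llama backpack mesh, rigging four dangling balloons for secondary motion on the pack.
# A box body covers the pack, with four two-body chains of a small capsule line and a sphere balloon.
# Each line pivots on the pack with free swing and twist, while each balloon joins its line with angular limits, tightest in twist.
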PhysicsAsset /Game/Accessories/FORT_Backpacks/F_MED_TechLlama_Backpack/Meshes/F_MED_Tech_Llama_Pack_Physics
bodies {
 dyn_balloonLine_01 [{shapeType=Capsule, length=1, radius=1, center=<1 0 0>, rotation=<rot 90 0 0>}]
 dyn_balloonLine_02 [{shapeType=Capsule, length=1, radius=1, center=<1 0 0>, rotation=<rot 90 0 0>}]
 dyn_balloonLine_03 [{shapeType=Capsule, length=1, radius=1, center=<1 0 0>, rotation=<rot 90 0 0>}]
 dyn_balloonLine_04 [{shapeType=Capsule, length=1, radius=1, center=<1 0 0>, rotation=<rot 90 0 0>}]
 dyn_balloon_01 [{shapeType=Sphere, radius=7.0798, center=<3 0 0>}]
 dyn_balloon_02 [{shapeType=Sphere, radius=6.8253, center=<7.0885 0 0>}]
 dyn_balloon_03 [{shapeType=Sphere, radius=4, center=<3 0 0>}]
 dyn_balloon_04 [{shapeType=Sphere, radius=5.9999, center=<2.5 0 0>}]
 pack [{shapeType=Box, extentX=35, extentY=4, extentZ=40, center=<0 7 7.5>}]
}
constraints [
 {
  bone1Name=dyn_balloonLine_01
  bone2Name=pack
  swing1LimitDegrees=45
  swing1Motion=Free
  swing2LimitDegrees=45
  swing2Motion=Free
  twistLimitDegrees=45
  twistMotion=Free
 }
 {bone1Name=dyn_balloon_01, bone2Name=dyn_balloonLine_01, swing1LimitDegrees=45, swing2LimitDegrees=45, twistLimitDegrees=5}
 {
  bone1Name=dyn_balloonLine_02
  bone2Name=pack
  swing1LimitDegrees=45
  swing1Motion=Free
  swing2LimitDegrees=45
  swing2Motion=Free
  twistLimitDegrees=45
  twistMotion=Free
 }
 {bone1Name=dyn_balloon_02, bone2Name=dyn_balloonLine_02, swing1LimitDegrees=45, swing2LimitDegrees=45, twistLimitDegrees=5}
 {
  bone1Name=dyn_balloonLine_03
  bone2Name=pack
  swing1LimitDegrees=45
  swing1Motion=Free
  swing2LimitDegrees=45
  swing2Motion=Free
  twistLimitDegrees=45
  twistMotion=Free
 }
 {bone1Name=dyn_balloon_03, bone2Name=dyn_balloonLine_03, swing1LimitDegrees=45, swing2LimitDegrees=45, twistLimitDegrees=5}
 {
  bone1Name=dyn_balloonLine_04
  bone2Name=pack
  swing1LimitDegrees=45
  swing1Motion=Free
  swing2LimitDegrees=45
  swing2Motion=Free
  twistLimitDegrees=45
  twistMotion=Free
 }
 {bone1Name=dyn_balloon_04, bone2Name=dyn_balloonLine_04, swing1LimitDegrees=45, swing2LimitDegrees=45, twistLimitDegrees=5}
]
